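# Translucent unlit water material for the SlurpMonster skydiving contrail effect's hands, usable on mesh particles and Niagara ribbons.
# Fresnel color blends, with optional noise-driven NitroFlow color, combine with a fake matcap-style reflection and time-of-day tint.
# One panned distortion normal map perturbs reflection lookup and drives opacity and radial plus directional vertex displacement.
# Dynamic parameters offset panning and scale directional displacement; mobile skips depth fade and reduces displacement.
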Material /BRCosmetics/Athena/Effects/Materials/Skydiving/P_Contrails_SlurpMon/MT_SlurpMonster_HandsWater
graph {
 properties {
  bUsedWithMeshParticles=true
  bUsedWithNiagaraMeshParticles=true
  bUsedWithNiagaraRibbons=true
  blendMode=BLEND_Translucent
  pixelDepthOffsetMode=PDOM_Legacy
  shadingModel=MSM_Unlit
  usedShadingModels=Unlit
 }
 nodes [
  n1 = ScalarParameter() {defaultValue=1.3, parameterName=FresnelBMult}
  n2 = ScalarParameter() {defaultValue=8, parameterName=FresnelBExp}
  n3 = MaterialFunctionCall(
   Brightness=ScalarParameter() {defaultValue=1.3, parameterName=FresnelMult}
   Exponent=ScalarParameter() {defaultValue=2, parameterName=FresnelExp}
  ) {materialFunction=&MF_Fresnel_FX}
  n4 = LinearInterpolate(
   A=VectorParameter() {parameterName=ColorA, defaultValue=<rgba 0 0.1875 0.1037 1>}.RGB
   B=VectorParameter() {parameterName=ColorB, defaultValue=<rgba 0.0335 0.1033 0.1927 1>}.RGB
   Alpha=@n3.Fresnel
  )
  n5 = LinearInterpolate(
   A=VectorParameter() {parameterName=ColorC, defaultValue=<rgba 0 0.1875 0.1037 1>}.RGB
   B=VectorParameter() {parameterName=ColorD, defaultValue=<rgba 0.0335 0.1033 0.1927 1>}.RGB
   Alpha=@n3.Fresnel
  )
  n6 = TextureSample() {samplerType=SAMPLERTYPE_Grayscale, texture=/Game/UI/Frontend/ItemShop/Art/Rumble/LowResBlurredNoise_Rumble}
  n7 = Add(
   A=Multiply(
    A=MaterialFunctionCall(Brightness=@n1, Exponent=@n2) {materialFunction=&MF_Fresnel_FX}.Fresnel Inverted
    B=@n4
   )
   B=LinearInterpolate(
    A=@n4
    B=LinearInterpolate(A=@n4, B=@n5, Alpha=@n6.RGB)
    Alpha=ScalarParameter() {group=NitroFlow_Options, parameterName=UseNitroFlow_Color, sortPriority=0}
   )
  )
  n8 = TextureCoordinate()
  n9 = ScalarParameter() {defaultValue=0.6, parameterName=DistTile}
  n10 = Time() {bOverride_Period=true, period=64}
  n11 = DynamicParameter() {paramNames [rand 0-1, WPO_Amnt, Param3, Param4]}
  n12 = ScalarParameter() {defaultValue=2, parameterName=DistSpeed}
  n13 = Constant() {r=0.1}
  n14 = TextureSample(UVs=Panner(Coordinate=Multiply(A=@n8, B=@n9), Time=Add(A=@n10, B=@n11.rand 0-1), Speed=AppendVector(A=@n12, B=@n13)) {speedX=-1, speedY=0.1}) {samplerType=SAMPLERTYPE_Normal, texture=&T_Weap_Distort_N_02}
  n15 = Transform(None=@n14.RGB)
  n16 = ScalarParameter() {defaultValue=0.5, parameterName=NoiseSTr}
  n17 = Transform(None=Add(A=Multiply(A=@n15, B=@n16), B=VertexNormalWS())) {transformSourceType=TRANSFORMSOURCE_World, transformType=TRANSFORM_View}
  n18 = ComponentMask(None=Add(A=Multiply(A=@n17) {constB=0.5}) {constB=0.5}) {g=true, r=true}
  n19 = Multiply(
   A=MaterialFunctionCall(
    In=TextureSample(UVs=@n18) {texture=/Game/Packages/Fortress_SharedMaterials/Textures/T_Ravens_SpecularReflection}.G
    Contrast=ScalarParameter() {parameterName=ReflectionContrast}
   ) {materialFunction=&CheapContrast}.Result
   B=VectorParameter() {parameterName=ReflectionColor, defaultValue=<rgba 0 0.3 4 1>}.RGB
  )
  n20 = VectorParameter() {group=TOD, parameterName=TOD_Morning, defaultValue=<rgba 1 1 1 1>}
  n21 = VectorParameter() {group=TOD, parameterName=TOD_Afternoon, defaultValue=<rgba 1 1 1 1>}
  n22 = VectorParameter() {group=TOD, parameterName=TOD_Evening, defaultValue=<rgba 1 1 1 1>}
  n23 = VectorParameter() {group=TOD, parameterName=TOD_Night, defaultValue=<rgba 1 1 1 1>}
  MaterialFunctionCall(Emissive=Multiply(
    A=Add(A=@n7, B=@n19)
    B=MaterialFunctionCall(Color_Morning=@n20.RGB, Color_Day=@n21.RGB, Color_Evening=@n22.RGB, Color_Night=@n23.RGB) {materialFunction=&Color_Time_of_Day}.Result
   )) {materialFunction=&MF_FX_EmissiveAdjustments}
  n24 = VertexColor()
  n25 = Constant() {r=1}
  n26 = ScalarParameter() {defaultValue=0.75, parameterName=OpacBase}
  n27 = MaterialFunctionCall(In=Add(A=Multiply(A=@n14.G) {constB=0.5}) {constB=0.5}, Contrast=ScalarParameter() {defaultValue=1, parameterName=contrast}) {materialFunction=&CheapContrast}
  n28 = Saturate(None=Multiply(A=StaticSwitchParameter(True=@n24.G, False=@n25) {parameterName=MultiplyVertColorOpacity}, B=Add(A=@n26, B=@n27.Result)))
  n29 = ScalarParameter() {defaultValue=0.95, parameterName=OpacEnd}
  n30 = Constant() {r=1}
  n31 = ShadingPathSwitch(Default=DepthFade(FadeDistance=ScalarParameter() {defaultValue=10, parameterName=DFDist}), Mobile=@n30) {desc=Migrated from FeatureLevelSwitch}
  Multiply(A=Multiply(A=Multiply(A=@n28, B=@n29), B=ParticleColor().A), B=QualitySwitch(Default=@n31, Low=@n30, High=@n31, Medium=@n31))
  n32 = Subtract(A=WorldPosition().XYZ, B=ObjectPositionWS())
  n33 = ScalarParameter() {defaultValue=-1, parameterName=WPOAmnt2LOW}
  n34 = ScalarParameter() {defaultValue=3, parameterName=WPOAmnt2}
  n35 = VertexNormalWS()
  n36 = VectorParameter() {parameterName=WPO_Direction, defaultValue=<rgba 0 0 1 1>}
  n37 = StaticSwitchParameter(True=Transform(None=@n36.RGB) {transformSourceType=TRANSFORMSOURCE_Local}, False=@n36.RGB) {parameterName=LocalWPO}
  n38 = Multiply(A=Saturate(None=DotProduct(A=@n35, B=@n37)), B=@n37)
  n39 = ScalarParameter() {defaultValue=16, parameterName=WPOAmnt}
  n40 = Add(
   A=Multiply(A=Normalize(VectorInput=@n32), B=LinearInterpolate(A=@n33, B=@n34, Alpha=@n27.Result))
   B=Multiply(A=Multiply(A=@n38, B=@n27.Result), B=Multiply(A=@n11.WPO_Amnt, B=@n39))
  )
  ShadingPathSwitch(Default=@n40, Deferred=@n40, Mobile=Multiply(A=@n40) {constB=0.1}) {desc=Migrated from FeatureLevelSwitch}
 ]
}
